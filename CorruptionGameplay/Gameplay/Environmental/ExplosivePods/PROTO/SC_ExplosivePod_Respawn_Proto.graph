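# Prototype sound cue for the respawn pop of an explosive pod, an environmental object in Corruption gameplay.
# A single wave player feeds the pop wave straight to the output with boosted volume.
# The cue overrides attenuation for distance falloff and concurrency to prevent new instances once the limit is reached.
# It is assigned to the non-critical in-game UI sound class, despite being an environmental sound.
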SoundCue /CorruptionGameplay/Gameplay/Environmental/ExplosivePods/PROTO/SC_ExplosivePod_Respawn_Proto
root=WavePlayer_1
nodes [{id=WavePlayer_1, properties {soundWaveAssetPtr=&/explosive_pod_respawn_pop}}]
properties {
 bOverrideAttenuation=true
 bOverrideConcurrency=true
 maxDistance=3400
 soundClassObject=&Fort_UI_InGame_NonCritical
 volumeMultiplier=1.4
 attenuationOverrides {falloffDistance=3000}
 concurrencyOverrides {resolutionRule=PreventNew}
}
